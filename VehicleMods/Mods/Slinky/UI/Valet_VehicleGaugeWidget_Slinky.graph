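# Vehicle dashboard gauge widget for the Slinky mod, adding a radial jump-charge meter, glow and icon to the shared gauge base.
# When shown, it finds the vehicle's VMC_Slinky component and, for Valet-tagged vehicles, binds its Increase and Clear Charge events, unbinding when hidden.
# Charge handlers drive the fill parameter on both meter and glow dynamic materials, remapping charge above the component's minimum percentage and zeroing it on clear.
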WidgetBlueprint /VehicleMods/Mods/Slinky/UI/Valet_VehicleGaugeWidget_Slinky
graphs [
 EventGraph {nodes [
   On Clear Charge Jump() {is=title K2Node_CustomEvent}
   > SetScalarParameterValue(
    self=[
     Get ImageJumpMaterial() -> (ImageJumpMaterial:Material Instance Dynamic Object Reference)
     Get JumpGlowMaterial() -> (JumpGlowMaterial:Material Instance Dynamic Object Reference)
    ]
    ParameterName=Get Jump Charge MID Parameter() -> (Jump Charge MID Parameter:Name=None)
    Value=0
   )
   n1 = Event OnWidgetVisibilityChanged!() -> (OutputDelegate, then, bNewVisibility:bool) {is=title K2Node_Event}
   > n2 = Branch(Condition=@n1.bNewVisibility)
   n3 = GetVehicle!(self:Fort Athena Vehicle Dashboard Widget Object Reference) -> (ReturnValue:Fort Athena Vehicle Object Reference)
   > n4 = Set Slinky Mod Component(Slinky Mod Component:VMC Slinky Object Reference=GetComponentByClass!(self:Actor Object Reference=@n3, ComponentClass:Actor Component Class Reference=/VehicleMods/Mods/Slinky/VMC_Slinky.VMC_Slinky_C) -> (ReturnValue:VMC Slinky Object Reference)) -> (Output_Get:VMC Slinky Object Reference)
   n5 = GetVehicleTags!(self:Fort Vehicle Interface Interface=@n3) -> (ReturnValue:"Gameplay Tag Container Structure (by ref)")
   n6 = MakeGameplayTagContainerFromArray!(GameplayTags:Array of Gameplay Tag Structures=Make Array!("[0]"=(TagName="Vehicle.Valet"), "[1]":Gameplay Tag Structure=(TagName="Athena.Vehicle.Valet")) -> (Array)) -> (ReturnValue:Gameplay Tag Container Structure)
   > Branch(Condition=AND Boolean(
     A=IsValid(Object=@n4)
     B=HasAnyTags!(
      TagContainer:"Gameplay Tag Container Structure (by ref)"=@n5
      OtherContainer:"Gameplay Tag Container Structure (by ref)"=@n6
      bExactMatch:bool=false
     ) -> (ReturnValue:bool)
    ))
   n7 = Create Event()
   > Bind Event to Increase Charge(self:VMC Slinky Object Reference=@n4, Delegate=@n7) {is=title K2Node_AddDelegate}
   n8 = Create Event()
   > Bind Event to Clear Charge(self:VMC Slinky Object Reference=@n4, Delegate=@n8) {is=title K2Node_AddDelegate}
   > SetScalarParameterValue(self=[Get JumpGlowMaterial(), Get ImageJumpMaterial()], ParameterName=PercentFill, Value=0)
   Unbind Event from Increase Charge(execute=@n2.else, self:VMC Slinky Object Reference=@n4, Delegate=@n7) {is=title K2Node_RemoveDelegate}
   > Unbind Event from Clear Charge(self:VMC Slinky Object Reference=@n4, Delegate=@n8) {is=title K2Node_RemoveDelegate}
   Event Construct()
   > SetVisibility(InVisibility=Collapsed)
   > n9 = GetDynamicMaterial(self=Get Image_JumpMeter() -> (Image_JumpMeter:Image Object Reference))
   > Set ImageJumpMaterial(ImageJumpMaterial:Material Instance Dynamic Object Reference=@n9) -> (Output_Get:Material Instance Dynamic Object Reference)
   > n10 = GetDynamicMaterial(self=Get Image_JumpGlow() -> (Image_JumpGlow:Image Object Reference))
   > Set JumpGlowMaterial(JumpGlowMaterial:Material Instance Dynamic Object Reference=@n10) -> (Output_Get:Material Instance Dynamic Object Reference)
  ]}
 OnChargingJump {nodes [
   n1 = OnChargingJump!() -> (then, Charge:double=~) {is=title K2Node_FunctionEntry}
   > n2 = Get!(execute:Exec) -> (then:Exec, else:Exec, Slinky Mod Component:VMC Slinky Object Reference)
   n3 = Get Min Jump Charge Percentage!(self:VMC Slinky Object Reference=@n2) -> (Min Jump Charge Percentage:double)
   > n4 = Set Adjusted Total Percentage(Adjusted Total Percentage:double=float - float(A=1, B=@n3)) -> (Output_Get:double)
   > n5 = Branch(Condition="float > float"(A=@n4))
   > SetScalarParameterValue(
    self=[Get ImageJumpMaterial(), Get JumpGlowMaterial()]
    ParameterName=Get Jump Charge MID Parameter() -> (Jump Charge MID Parameter=None)
    Value=float / float(A=float - float(A=@n1, B=@n3), B=@n4)
   )
   Return Node(execute=@n5.else)
   SetScalarParameterValue(
    execute=@n2.else
    self=[Get ImageJumpMaterial(), Get JumpGlowMaterial()]
    ParameterName=Get Jump Charge MID Parameter() -> (Jump Charge MID Parameter=None)
    Value=@n1
   )
  ]}
]
variables [
 BoostEffectsVisible_0 {type=boolean}
 ImageJumpMaterial {type=MaterialInstanceDynamic}
 Jump Charge MID Parameter {type=string, default=PercentFill}
 JumpGlowMaterial {type=MaterialInstanceDynamic}
 Slinky Mod Component {type=VMC_Slinky_C, default=null}
]
widgets {
 info {
  inheritedWidgetCount=10
  parentClass=/Game/Athena/HUD/Vehicles/Dashboard/WBP_VehicleGaugeBase.WBP_VehicleGaugeBase_C
  rootWidgetClass=&Overlay
  widgetCount=18
 }
 widgets [
  Overlay_2 {widgetClassPath=&Overlay, widget=Overlay_2}
  Overlay_Gauge {
   widgetClassPath=&Overlay
   bInherited=true
   parent=Overlay_2
   widget=Overlay_Gauge
   properties {visibility=HitTestInvisible}
   slot {class=OverlaySlot, properties {horizontalAlignment=HAlign_Right, verticalAlignment=VAlign_Bottom}}
  }
  Overlay_SpeedCluster {
   widgetClassPath=&Overlay
   parent=Overlay_Gauge
   widget=Overlay_SpeedCluster
   properties {renderTransform {angle=0, scale=<1 1>, shear=<0 0>, translation=<14 16>}}
   slot {class=OverlaySlot, properties {horizontalAlignment=HAlign_Right, verticalAlignment=VAlign_Bottom, padding=<ltrb 0 0 45 0>}}
  }
  Overlay_Speedometer {widgetClassPath=&Overlay, bIsVariable=true, parent=Overlay_SpeedCluster, widget=Overlay_Speedometer, slot {class=OverlaySlot}}
  SizeholderBox {
   widgetClassPath=&SizeBox
   parent=Overlay_Speedometer
   widget=SizeholderBox
   properties {bOverride_HeightOverride=true, bOverride_WidthOverride=true, heightOverride=330, widthOverride=330}
   slot {class=OverlaySlot}
  }
  Overlay_JumpMeter {
   widgetClassPath=&Overlay
   bIsVariable=true
   parent=Overlay_SpeedCluster
   widget=Overlay_JumpMeter
   properties {visibility=HitTestInvisible, renderTransform {angle=85, scale=<1 1>, shear=<0 0>, translation=<-10 -50>}}
   slot {class=OverlaySlot, properties {horizontalAlignment=HAlign_Center, verticalAlignment=VAlign_Center, padding=<ltrb -200 -200 -200 -200>}}
  }
  Image_JumpGlow {
   widgetClassPath=&Image
   bIsVariable=true
   parent=Overlay_JumpMeter
   widget=Image_JumpGlow
   properties {
    visibility=SelfHitTestInvisible
    brush {
     bIsDynamicallyLoaded=false
     drawAs=Image
     imageType=NoImage
     mirroring=NoMirror
     resourceName=None
     resourceObject=/Game/Athena/HUD/Vehicles/Materials/M_UI_ValetMeterGlow_Inst
     tiling=NoTile
     imageSize=<340 340>
     margin=<ltrb 0 0 0 0>
     outlineSettings {
      bUseBrushTransparency=false
      roundingType=HalfHeightRadius
      width=0
      color {colorUseRule=UseColor_Specified, specifiedColor=<rgba 0 0 0 0>}
      cornerRadii=<0 0 0 0>
     }
     tintColor {colorUseRule=UseColor_Specified, specifiedColor=<rgba 1 1 1 1>}
     uVRegion {bIsValid=false, max=<0 0>, min=<0 0>}
    }
   }
   slot {class=OverlaySlot, properties {horizontalAlignment=HAlign_Center, verticalAlignment=VAlign_Center, padding=<ltrb -200 -200 -200 -200>}}
  }
  Image_JumpMeter {
   widgetClassPath=&Image
   bIsVariable=true
   parent=Overlay_JumpMeter
   widget=Image_JumpMeter
   properties {
    visibility=SelfHitTestInvisible
    brush {
     bIsDynamicallyLoaded=false
     drawAs=Image
     imageType=NoImage
     mirroring=NoMirror
     resourceName=None
     resourceObject=/Game/Athena/HUD/Vehicles/Materials/M_UI_ValetSimpleMeter_Boost
     tiling=NoTile
     imageSize=<320 320>
     margin=<ltrb 0 0 0 0>
     outlineSettings {
      bUseBrushTransparency=false
      roundingType=HalfHeightRadius
      width=0
      color {colorUseRule=UseColor_Specified, specifiedColor=<rgba 0 0 0 0>}
      cornerRadii=<0 0 0 0>
     }
     tintColor {colorUseRule=UseColor_Specified, specifiedColor=<rgba 1 1 1 1>}
     uVRegion {bIsValid=false, max=<0 0>, min=<0 0>}
    }
   }
   slot {class=OverlaySlot}
  }
  Image_JumpIcon {
   widgetClassPath=&Image
   parent=Overlay_JumpMeter
   widget=Image_JumpIcon
   properties {
    visibility=SelfHitTestInvisible
    brush {
     bIsDynamicallyLoaded=false
     drawAs=Image
     imageType=FullColor
     mirroring=NoMirror
     resourceName=None
     resourceObject=&/T_UI_Slinky_Jump
     tiling=NoTile
     imageSize=<60 60>
     margin=<ltrb 0 0 0 0>
     outlineSettings {
      bUseBrushTransparency=false
      roundingType=HalfHeightRadius
      width=0
      color {colorUseRule=UseColor_Specified, specifiedColor=<rgba 0 0 0 0>}
      cornerRadii=<0 0 0 0>
     }
     tintColor {colorUseRule=UseColor_Specified, specifiedColor=<rgba 1 1 1 1>}
     uVRegion {bIsValid=false, max=<0 0>, min=<0 0>}
    }
    renderTransform {angle=-90, scale=<1 1>, shear=<0 0>, translation=<60 140>}
   }
   slot {class=OverlaySlot, properties {verticalAlignment=VAlign_Center, padding=<ltrb -64 -64 -64 -64>}}
  }
  Widget_FuelGauge {widgetClassPath=/Script/FortniteUI.FortVehicleFuelGauge, bInherited=true, widget=null}
  HorizontalBox_FuelCluster {widgetClassPath=&HorizontalBox, bInherited=true, widget=null}
  Image_Speedometer {widgetClassPath=&Image, bInherited=true, widget=null}
  Image_BoostGlow {widgetClassPath=&Image, bInherited=true, widget=null}
  Image_BoostMeter {widgetClassPath=&Image, bInherited=true, widget=null}
  Overlay_BoostSparks {widgetClassPath=&Overlay, bInherited=true, widget=null}
  NumericText_Speed {widgetClassPath=/Script/CommonUI.CommonNumericTextBlock, bInherited=true, widget=null}
  Text_ItemName {widgetClassPath=&CommonTextBlock, bInherited=true, widget=null}
  TutorialHighlight {widgetClassPath=&TutorialHighlight}
 ]
}
